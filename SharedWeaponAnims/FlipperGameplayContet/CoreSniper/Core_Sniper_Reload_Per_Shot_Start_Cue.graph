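# Sound cue for the Core Sniper's per-shot reload start, played as weapon foley.
# A mixer layers the reload start wave with the reload end wave, which follows after a fixed half-second delay.
# Each wave passes through its own modulator for volume variation.
# The cue uses shared weapon reload attenuation and the weapon foley sound class, overrides concurrency, and has its debug flag enabled.
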SoundCue /SharedWeaponAnims/FlipperGameplayContet/CoreSniper/Core_Sniper_Reload_Per_Shot_Start_Cue
root=Mixer_0
nodes [
 {id=Delay_0, children [Modulator_1], properties {delayMax=0.5, delayMin=0.5}}
 {id=Mixer_0, children [Modulator_0,Delay_0], properties {inputVolume [1,1]}}
 {id=Modulator_0, children [WavePlayer_29], properties {volumeMax=1.35}}
 {id=Modulator_1, children [WavePlayer_0], properties {volumeMax=1, volumeMin=0.6}}
 {id=WavePlayer_0, properties {soundWaveAssetPtr=&/Sniper_CoreSniper_Reload_Per_Shot_End_01}}
 {
  id=WavePlayer_29
  properties {soundWaveAssetPtr=&/Sniper_CoreSniper_Reload_Per_Shot_Start_01}
 }
]
properties {
 attenuationSettings=&Fort_Weapon_Reload_Att
 bDebug=true
 bOverrideConcurrency=true
 maxDistance=2250
 soundClassObject=&Fort_WeaponFoley
 volumeMultiplier=0.501
}
